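# Sound cue for equipping the Sunrose HandCannon, reusing a single pistol foley wave.
# The wave passes through a Modulator that randomly pitches it slightly down, under a root Delay that holds playback back by a fixed quarter second.
# It plays at reduced volume with weapon reload attenuation and the third-person ranged-weapon fire sound class.
# A concurrency override prevents new instances once the limit is reached.
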SoundCue /WeaponsUpdated/Gameplay/HandCannon/Sounds/Sunrose_HandCannon_Equip_Cue
root=Delay_0
nodes [
 {id=Delay_0, children [Modulator_1], properties {delayMax=0.25, delayMin=0.25}}
 {id=Modulator_1, children [WavePlayer_10], properties {pitchMax=0.95, pitchMin=0.85}}
 {id=WavePlayer_10, properties {soundWaveAssetPtr=/Game/Sounds/Fort_Weapons/pistol/pistol_foley/auto_sdk_end_01}}
]
properties {
 attenuationSettings=&Fort_Weapon_Reload_Att
 bOverrideConcurrency=true
 maxDistance=2250
 soundClassObject=&Fort_WeaponRanged_Fire_3P_PSM
 volumeMultiplier=0.4
 concurrencyOverrides {resolutionRule=PreventNew}
}
