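# Sound cue for the opening stage of the Sunrose hand cannon reload.
# A mixer layers a shared pistol eject foley wave with a Sunrose-specific sweetener, each passing through its own modulator.
# It plays at reduced volume with weapon reload attenuation and a concurrency override that blocks new instances at the limit.
# Its sound class is the third-person ranged weapon fire class, despite being a reload sound.
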SoundCue /WeaponsUpdated/Gameplay/HandCannon/Sounds/Sunrose_HandCannon_Reload_Open_Cue
root=Mixer_0
nodes [
 {id=Mixer_0, children [Modulator_0,Modulator_1], properties {inputVolume [1,1]}}
 {id=Modulator_0, children [WavePlayer_11]}
 {id=Modulator_1, children [WavePlayer_0]}
 {id=WavePlayer_0, properties {soundWaveAssetPtr=&/Sunrose_HandCannon_Reload_Open_Sweetener}}
 {id=WavePlayer_11, properties {soundWaveAssetPtr=/Game/Sounds/Fort_Weapons/pistol/pistol_foley/auto_sdk_eject_01}}
]
properties {
 attenuationSettings=&Fort_Weapon_Reload_Att
 bOverrideConcurrency=true
 maxDistance=2250
 soundClassObject=&Fort_WeaponRanged_Fire_3P_PSM
 volumeMultiplier=0.4
 concurrencyOverrides {resolutionRule=PreventNew}
}
